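# Niagara effect for the SlopeTramp emote's pancake splash: a short-lived spray of droplets and brown chunks.
# Two sprite emitters, DropsDir and CookieChunks, each burst particles from a small sphere, throw them mainly along +Y, then pull them down with gravity while they spin and shrink.
# DropsDir adds SubUV, alpha fade and animated dynamic material parameters.
# Both emitters tint color by time of day through FNTimeOfDayColor.
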NiagaraSystem /BRCosmetics/Animation/Game/MainPlayer/Emotes/SlopeTramp/FX/NS_Emote_SlopeTramp_PancakeSplash
dependencies {
 usedDataInterfaces [&NiagaraDataInterfaceCurve, &NiagaraDataInterfaceMeshRendererInfo]
 usedDynamicInputs []
 usedModules [&SystemState]
}
emitters {
 CookieChunks {
  emitterUpdateScript [
   &EmitterState()
   &SpawnBurst_Instantaneous(Spawn Count=10, Use Spawn Probability=false, UseLoopCountLimit=false)
  ]
  particleSpawnScript [
   &InitializeParticle(
    Color=<rgba 0.3385 0.1836 0.0838 1>
    Lifetime Max=0.4
    Lifetime Min=0.3
    Lifetime Mode=::NewEnumerator1
    Mesh Scale Mode=::NewEnumerator3
    Mesh Uniform Scale=&Multiply_Float(
     A=0.4
     B=&FN_FrontEndFloatModifier(
      Gameplay=&RandomRangeFloat(Minimum=1, Maximum=2)
      Front End=&RandomRangeFloat(Minimum=0.5, Maximum=1)
     )
    )
    Sprite Rotation Mode=::NewEnumerator2
    Sprite Size Mode=::NewEnumerator3
    Sprite UV Mode=::NewEnumerator3
    Uniform Sprite Size=&MakeFloatFromVector2d(Vector 2D=&Multiply_Vector2D_ByFloat(
      Float=&UniformAOrBFloat(
       A=&RandomRangeFloat(Minimum=2.5, Maximum=5.5)
       B=6.5
       Distribution Weight=0.1
       Evaluation Type=::NewEnumerator0
       Recalculate Random Each Loop=true
       Override Seed=false
      )
      Vector2D=&RandomRangeVector2D(Minimum=<0.8 1>, Maximum=<0.8 2>)
     ))
    UsePositionOffset=false
    Write Material Random=true
   )
   &ShapeLocation(MaskSpawnGroup=false, Override Seed=false, Sphere Radius=3, UseOwnerScale=true)
   &AddVelocity(Velocity=&RandomRangeVector(Minimum=<-10 40 -30>, Maximum=<30 80 30>), Velocity Speed Scale=1.4)
  ]
  particleUpdateScript [
   &ParticleState()
   &SpriteRotationRate(Rotation Rate=&RandomRangeFloat(Minimum=5, Maximum=25))
   &ScaleSpriteSize(Uniform Curve Sprite Scale=DataInterface() {propertyValues {Curve {keys [<at 0.6 1 RCIM_Cubic>, {arriveTangent=-6.308, interpMode=RCIM_Cubic, leaveTangent=-6.308, tangentMode=RCTM_User, time=1, value=0}]}}})
   &ScaleMeshSize(Scale Factor=&VectorFromFloat(Value=&FloatFromCurve(FloatCurve=DataInterface() {propertyValues {Curve {keys [<at 0.6 1 RCIM_Cubic>, {arriveTangent=-6.308, interpMode=RCIM_Cubic, leaveTangent=-6.308, tangentMode=RCTM_User, time=1, value=0}]}}})))
   &GravityForce(Gravity=&RandomRangeVector(Minimum=<0 0 -500>, Maximum=<0 0 -800>)) {module=GravityForce001}
   &SolveForcesAndVelocity(Clamp Velocity=false, Limit Acceleration=false)
   &ScaleColor(
    Scale RGB=&MakeVectorFromLinearColorRGB(Color=&FNTimeOfDayColor(
      Night Color=<rgba 0.7726 0.8287 1 1>
      Morning Color=<rgba 1 0.7838 0.6226 1>
      Day Color=<rgba 1 1 1 1>
      Evening Color=<rgba 1 0.691 0.4705 1>
     ))
    ScaleA=false
    ScaleRGB=true
   )
  ]
  rendererClasses [&NiagaraSpriteRendererProperties]
  renderers [{rendererClass=&NiagaraSpriteRendererProperties}]
 }
 DropsDir {
  emitterUpdateScript [
   &EmitterState(Life Cycle Mode=::NewEnumerator1, Loop Behavior=::NewEnumerator1, Loop Duration=0.35, UseLoopDelay=false)
   &SpawnBurst_Instantaneous(Spawn Count=10, Use Spawn Probability=false, UseLoopCountLimit=false)
  ]
  particleSpawnScript [
   &InitializeParticle(
    Lifetime Max=0.4
    Lifetime Min=0.3
    Lifetime Mode=::NewEnumerator1
    Sprite Rotation Mode=::NewEnumerator2
    Sprite Size=&Multiply_Vector2D_ByFloat(
     Float=&UniformAOrBFloat(
      A=&RandomRangeFloat(Minimum=2.5, Maximum=5.5)
      B=6.5
      Distribution Weight=0.1
      Evaluation Type=::NewEnumerator0
      Recalculate Random Each Loop=true
      Override Seed=false
     )
     Vector2D=&RandomRangeVector2D(Minimum=<0.8 1>, Maximum=<0.8 2>)
    )
    Sprite Size Mode=::NewEnumerator4
    UsePositionOffset=false
    Write Material Random=true
   ) {module=InitializeParticle001}
   &DynamicMaterialParameters(
    Index 0 Param 2=1
    Index 0 Param 3=1
    Index 0 Param 4=1
    Module.Param0WriteEnabled=true
    Module.Param1WriteEnabled=false
    Module.Param2WriteEnabled=false
    Module.Param3WriteEnabled=false
   )
   &ShapeLocation(MaskSpawnGroup=false, Override Seed=false, Sphere Radius=3, UseOwnerScale=true)
   &AddVelocity(Velocity=&RandomRangeVector(Minimum=<-10 40 -30>, Maximum=<30 80 30>), Velocity Speed Scale=1.4)
   /Niagara/Modules/Update/SubUV/SubUVAnimation(SubUV Animation Mode=::NewEnumerator1) {module=SubUVAnimation001}
  ]
  particleUpdateScript [
   &ParticleState()
   &ScaleColor(
    Scale Alpha=&FloatFromCurve(FloatCurve=DataInterface() {propertyValues {Curve {keys [<at 0.7037 1 RCIM_Cubic>,<at 1 0>]}}})
    Scale RGB=&MakeVectorFromLinearColorRGB(Color=&FNTimeOfDayColor(Night Color=<rgba 0.7583 0.9628 1.4 1>, Morning Color=<rgba 1.7 1.7 1.7 1>, Day Color=<rgba 1 1 1 1>, Evening Color=<rgba 1.65 1.65 1.65 1>))
    ScaleA=true
    ScaleRGB=true
   ) {module=ScaleColor001}
   &ScaleSpriteSize(
    Uniform Curve Scale=0.4
    Uniform Curve Sprite Scale=DataInterface() {propertyValues {Curve {keys [
        {arriveTangent=-4.5487, interpMode=RCIM_Cubic, time=0.33, value=1}
        {arriveTangent=-6.6107, interpMode=RCIM_Cubic, leaveTangent=-6.6107, tangentMode=RCTM_User, time=1, value=0}
       ]}}}
   ) {module=ScaleSpriteSize001}
   &GravityForce(Gravity=&RandomRangeVector(Minimum=<0 0 -500>, Maximum=<0 0 -800>)) {module=GravityForce001}
   &SolveForcesAndVelocity(Clamp Velocity=false, Limit Acceleration=false)
   &DynamicMaterialParameters(
    Index 0 Param 1=&RandomRangeFloat(Minimum=0, Maximum=1)
    Index 0 Param 2=&FloatFromCurve(FloatCurve=DataInterface() {propertyValues {Curve {keys [<at 0.35 0.45>,<at 1 0.9>]}}})
    Index 0 Param 3=&FloatFromCurve(FloatCurve=DataInterface() {propertyValues {Curve {keys [{arriveTangent=2.6651, interpMode=RCIM_Cubic, leaveTangent=2.6651, tangentMode=RCTM_User, time=0, value=0}, <at 1 1 RCIM_Cubic>]}}})
    Index 0 Param 4=&FloatFromCurve(FloatCurve=DataInterface() {propertyValues {Curve {keys [<at 0.5 0 RCIM_Cubic>, {arriveTangent=2.332, interpMode=RCIM_Cubic, leaveTangent=2.332, tangentMode=RCTM_User, time=0.75, value=1}]}}})
    Module.Param0WriteEnabled=true
    Module.Param1WriteEnabled=true
    Module.Param2WriteEnabled=false
    Module.Param3WriteEnabled=true
   ) {module=DynamicMaterialParameters002}
   &SpriteRotationRate(Rotation Rate=&Multiply_Float(
     A=&Multiply_Float(A=&RandomRangeFloat(Minimum=1, Maximum=0.35), B=$Particles.Random)
     B=&FloatFromCurve(
      FloatCurve=DataInterface() {propertyValues {Curve {keys [{arriveTangent=-2, interpMode=RCIM_Cubic, time=0.15, value=1}, {interpMode=RCIM_Cubic, leaveTangent=-8.0089, time=0.5, value=0}]}}}
      Scale Curve=800
     )
    ))
  ]
  rendererClasses [&NiagaraSpriteRendererProperties]
  renderers [{rendererClass=&NiagaraSpriteRendererProperties}]
 }
}
summary {
 systemName=NS_Emote_SlopeTramp_PancakeSplash
 emitters [
  {emitterName=DropsDir, rendererClasses [&NiagaraSpriteRendererProperties]}
  {emitterName=CookieChunks, rendererClasses [&NiagaraSpriteRendererProperties]}
 ]
 userVariables [
  User.bIsFrontEnd {type=NiagaraBool}
  User.bIsFrontEndPreview {type=NiagaraBool}
 ]
}
