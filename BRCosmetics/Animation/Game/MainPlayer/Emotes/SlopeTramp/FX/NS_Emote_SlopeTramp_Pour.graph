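# Niagara system for the pouring liquid effect in the SlopeTramp emote.
# The Stream emitter spawns a ribbon with curve-driven launch velocity and acceleration, with User.VelocityIndx switching vertical velocity profiles and skeletal position enabled in the ribbon update.
# Ribbon width follows link order, blending between two profiles over system age before alpha fades out, and color is tinted by time of day.
# A secondary sprite Drops emitter with gravity is present but disabled.
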NiagaraSystem /BRCosmetics/Animation/Game/MainPlayer/Emotes/SlopeTramp/FX/NS_Emote_SlopeTramp_Pour
dependencies {
 usedDataInterfaces [&NiagaraDataInterfaceCurve, /Script/Niagara.NiagaraDataInterfaceSimpleCounter]
 usedDynamicInputs []
 usedModules [&SystemState]
}
emitters {
 Drops {
  bEnabled=false
  emitterUpdateScript [
   &EmitterState(Life Cycle Mode=::NewEnumerator1, Loop Behavior=::NewEnumerator1, Loop Duration=0.35, UseLoopDelay=false)
   &SpawnRate(
    SpawnRate=&FloatFromCurve(
     FloatCurve=DataInterface() {propertyValues {Curve {keys [
         <at 0 0.2449 RCIM_Cubic>
         {arriveTangent=1.5669, interpMode=RCIM_Cubic, leaveTangent=1.5669, time=0.1249, value=1}
         {interpMode=RCIM_Cubic, tangentMode=RCTM_User, time=0.4819, value=1}
         {interpMode=RCIM_Cubic, tangentMode=RCTM_User, time=0.8938, value=0}
        ]}}}
     CurveIndex=$Emitter.NormalizedLoopAge
     Scale Curve=34
    )
    Use Spawn Probability=true
   )
  ]
  particleSpawnScript [
   &InitializeParticle(
    Lifetime Max=0.65
    Lifetime Min=0.55
    Lifetime Mode=::NewEnumerator1
    Sprite Rotation Mode=::NewEnumerator2
    Sprite Size=&Multiply_Vector2D_ByFloat(
     Float=&UniformAOrBFloat(
      A=&RandomRangeFloat(Minimum=2.5, Maximum=5.5)
      B=6.5
      Distribution Weight=0.1
      Evaluation Type=::NewEnumerator0
      Recalculate Random Each Loop=true
      Override Seed=false
     )
     Vector2D=&RandomRangeVector2D(Minimum=<0.8 1>, Maximum=<0.9 2>)
    )
    Sprite Size Mode=::NewEnumerator4
    UsePositionOffset=false
    Write Material Random=true
   ) {module=InitializeParticle001}
   &DynamicMaterialParameters(
    Index 0 Param 2=1
    Index 0 Param 3=1
    Index 0 Param 4=1
    Module.Param0WriteEnabled=true
    Module.Param1WriteEnabled=false
    Module.Param2WriteEnabled=false
    Module.Param3WriteEnabled=false
   )
   &ShapeLocation(MaskSpawnGroup=false, Override Seed=false, Sphere Radius=1, UseOwnerScale=true)
   &AddVelocity(Velocity=&RandomRangeVector(Minimum=<-9 -9 15>, Maximum=<9 9 20>), Velocity Speed Scale=3)
   /Niagara/Modules/Update/SubUV/SubUVAnimation(SubUV Animation Mode=::NewEnumerator1) {module=SubUVAnimation001}
  ]
  particleUpdateScript [
   &ParticleState()
   &ScaleColor(
    Scale Alpha=&FloatFromCurve(FloatCurve=DataInterface() {propertyValues {Curve {keys [<at 0.8103 1 RCIM_Cubic>,<at 1 0>]}}})
    Scale RGB=&MakeVectorFromLinearColorRGB(Color=&FNTimeOfDayColor(Night Color=<rgba 0.7583 0.9628 1.4 1>, Morning Color=<rgba 1.7 1.7 1.7 1>, Day Color=<rgba 1 1 1 1>, Evening Color=<rgba 1.65 1.65 1.65 1>))
    ScaleA=true
    ScaleRGB=true
   ) {module=ScaleColor001}
   &ScaleSpriteSize(
    Uniform Curve Scale=0.7
    Uniform Curve Sprite Scale=DataInterface() {propertyValues {Curve {keys [
        {arriveTangent=-4.5487, interpMode=RCIM_Cubic, time=0.33, value=1}
        {arriveTangent=-6.6107, interpMode=RCIM_Cubic, leaveTangent=-6.6107, tangentMode=RCTM_User, time=1, value=0}
       ]}}}
   ) {module=ScaleSpriteSize001}
   &GravityForce(Gravity=&RandomRangeVector(Minimum=<0 0 -180>, Maximum=<0 0 -200>)) {module=GravityForce001}
   &SolveForcesAndVelocity(Clamp Velocity=false, Limit Acceleration=false)
   &DynamicMaterialParameters(
    Index 0 Param 1=&RandomRangeFloat(Minimum=0, Maximum=1)
    Index 0 Param 2=&FloatFromCurve(FloatCurve=DataInterface() {propertyValues {Curve {keys [<at 0.42 0.45 RCIM_Cubic>,<at 1 0.9>]}}})
    Index 0 Param 3=&FloatFromCurve(FloatCurve=DataInterface() {propertyValues {Curve {keys [{arriveTangent=2.6651, interpMode=RCIM_Cubic, leaveTangent=2.6651, tangentMode=RCTM_User, time=0, value=0}, <at 1 1 RCIM_Cubic>]}}})
    Index 0 Param 4=&FloatFromCurve(FloatCurve=DataInterface() {propertyValues {Curve {keys [<at 0.5 0 RCIM_Cubic>, {arriveTangent=2.332, interpMode=RCIM_Cubic, leaveTangent=2.332, tangentMode=RCTM_User, time=0.75, value=1}]}}})
    Module.Param0WriteEnabled=true
    Module.Param1WriteEnabled=true
    Module.Param2WriteEnabled=false
    Module.Param3WriteEnabled=true
   ) {module=DynamicMaterialParameters002}
   &SpriteRotationRate(Rotation Rate=&Multiply_Float(
     A=&Multiply_Float(A=&RandomRangeFloat(Minimum=1, Maximum=0.35), B=$Particles.Random)
     B=&FloatFromCurve(
      FloatCurve=DataInterface() {propertyValues {Curve {keys [{arriveTangent=-2, interpMode=RCIM_Cubic, time=0.15, value=1}, {interpMode=RCIM_Cubic, leaveTangent=-8.0089, time=0.5, value=0}]}}}
      Scale Curve=800
     )
    ))
  ]
  rendererClasses [&NiagaraSpriteRendererProperties]
  renderers [{rendererClass=&NiagaraSpriteRendererProperties}]
 }
 Stream {
  emitterSpawnScript [/SetVariables(Emitter.RibbonCount=1, Emitter.RibbonSpawnRate=30) {isSetParameters=true}]
  emitterUpdateScript [
   &EmitterState(Inactive Response=::NewEnumerator1, Life Cycle Mode=::NewEnumerator1, Loop Behavior=::NewEnumerator1, Loop Duration=1, UseLoopDelay=false)
   &SpawnRate(
    SpawnRate=&FloatFromCurve(
     FloatCurve=DataInterface() {propertyValues {Curve {keys [<at 0 1 RCIM_Constant>,<at 0.5 0 RCIM_Cubic>]}}}
     CurveIndex=$Emitter.NormalizedLoopAge
     Scale Curve=30
    )
    Use Spawn Probability=false
   )
  ]
  particleSpawnScript [
   /Game/Effects/Niagara/Modules/Ribbons/FN_InitializeRibbon_v5(Color:LinearColor=<rgba 0.5781 0.5781 0.5781 1>, Custom Offset=true, Lifetime=0.92, Offset Amount=1, Ribbon Width=40.6523)
   /SetVariables(Particles.SpriteSize=&Vector2DFromFloat(Value=1)) {isSetParameters=true}
   /SetVariables(Particles.Counter:NiagaraFloat=&Multiply_Float(
     A=&MakeFloatFromInt(Int=/Niagara/DynamicInputs/DataInterface/GenerateIntFromCounter(SimpleCounter:NiagaraDataInterfaceSimpleCounter=DataInterface() {propertyValues {GpuSyncMode=None, InitialValue=0}}))
     B=&Divide_Float(A=1, B=$Emitter.RibbonSpawnRate)
    )) {isSetParameters=true}
   &AddVelocity(Velocity=&MakeVector(
     X=&FloatFromCurve(
      FloatCurve=DataInterface() {propertyValues {Curve {keys [
          {arriveTangent=1.7154, interpMode=RCIM_Cubic, leaveTangent=1.7154, tangentMode=RCTM_User, time=0, value=0.5}
          {arriveTangent=0.4772, arriveTangentWeight=0.1, interpMode=RCIM_Cubic, leaveTangent=0.4772, tangentMode=RCTM_User, time=0.8, value=1}
         ]}}}
      Scale Curve=-175
     )
     Y=-1
     Z=&Add_Float(
      A=0
      B=&Add_Float(
       A=&FloatFromCurve(FloatCurve=DataInterface() {propertyValues {Curve {keys [<at 0 1 RCIM_Cubic>,<at 1 0 RCIM_Cubic>]}}}, Scale Curve=2)
       B=&FloatFromCurve(FloatCurve=DataInterface() {propertyValues {Curve {keys [<at 0 1 RCIM_Cubic>,<at 0.7673 0 RCIM_Cubic>]}}}, Scale Curve=11.2)
      )
     )
    ))
  ]
  particleUpdateScript [
   &ParticleState()
   /SetVariables(
    Particles.NormalizeLoop:NiagaraFloat=&Divide_Float(A=&Subtract_Float(A=$Emitter.Age, B=$Emitter.CurrentLoopDelay), B=$Emitter.CurrentLoopDuration)
    Particles.RibbonLinkOrder:NiagaraFloat=$Particles.NormalizedAge
   ) {isSetParameters=true}
   &ScaleVelocity(Velocity Scale=&MakeVector(
     X=1
     Y=1
     Z=&MergeHelper_FloatSwitch(
      0=&FloatFromCurve(FloatCurve=DataInterface() {propertyValues {Curve {keys [
           <at 0.1053 9 RCIM_Cubic>
           {arriveTangent=-15.9965, interpMode=RCIM_Cubic, leaveTangent=-15.9965, tangentMode=RCTM_User, time=0.28, value=1}
           {arriveTangent=-4.4816, interpMode=RCIM_Cubic, leaveTangent=-4.4816, time=0.3852, value=0.845}
           <at 0.6816 -0.8 RCIM_Cubic>
          ]}}})
      1=&FloatFromCurve(
       FloatCurve=DataInterface() {propertyValues {Curve {keys [
           <at 0.1053 9 RCIM_Cubic>
           {arriveTangent=-15.9965, interpMode=RCIM_Cubic, leaveTangent=-15.9965, tangentMode=RCTM_User, time=0.3068, value=1}
           {arriveTangent=-4.0022, interpMode=RCIM_Cubic, leaveTangent=-4.0022, time=0.3852, value=0.845}
           <at 0.6816 -0.5 RCIM_Cubic>
          ]}}}
       Scale Curve=1.15
      )
      2=?
      3=?
      Index=&MakeIntFromFloat(Float=$User.VelocityIndx, Conversion Mode=::NewEnumerator1)
      Num=::NewEnumerator0
     )
    ))
   &AccelerationForce(
    Acceleration=&Add_Vector(
     A=<10 0 0>
     B=&MakeVector(
      X=&FloatFromCurve(
       FloatCurve=DataInterface() {propertyValues {Curve {keys [
           {arriveTangent=-2.2093, interpMode=RCIM_Cubic, leaveTangent=-2.2093, tangentMode=RCTM_User, time=0, value=1}
           {interpMode=RCIM_Cubic, tangentMode=RCTM_User, time=0.3, value=-1}
           <at 0.5 0>
          ]}}}
       Scale Curve=300
      )
      Y=0
      Z=0
     )
    )
    Coordinate Space=::Local
   )
   &SolveForcesAndVelocity(Clamp Velocity=false, Limit Acceleration=false)
   &FN_UpdateRibbon(
    Enable Skeletal Position=true
    Initial Color:LinearColor=$Particles.Initial.Color
    Initial RibbonWidth:NiagaraFloat=$Particles.Initial.RibbonWidth
    Sampled Position Coordinate Space:ENiagaraCoordinateSpace=::Local
    Sampled Rotation Coordinate Space:ENiagaraCoordinateSpace=::Local
    Sampled Velocity Coordinate Space:ENiagaraCoordinateSpace=::Local
   )
   &ScaleRibbonWidth(Ribbon Width Scale=&FloatFromCurve(
     FloatCurve=DataInterface() {propertyValues {Curve {keys [<at 0.6778 1 RCIM_Cubic>, {arriveTangent=0.3089, interpMode=RCIM_Cubic, leaveTangent=0.3089, tangentMode=RCTM_User, time=0.8126, value=0}]}}}
     CurveIndex=$System.Age
    )) {enabled=false, module=ScaleRibbonWidth001}
   &ScaleRibbonWidth(Ribbon Width Scale=&Lerp_Float(
     A=&FloatFromCurve(
      FloatCurve=DataInterface() {propertyValues {Curve {keys [
          <at 0.2286 0.2496 RCIM_Cubic>
          {arriveTangent=-0.0519, interpMode=RCIM_Cubic, leaveTangent=-0.0519, tangentMode=RCTM_User, time=0.3281, value=0.38}
          {arriveTangent=-0.596, interpMode=RCIM_Cubic, leaveTangent=-0.596, tangentMode=RCTM_User, time=0.4854, value=0.4}
          {arriveTangent=-0.7765, interpMode=RCIM_Cubic, leaveTangent=-0.7765, time=0.6092, value=0.2764}
          {arriveTangent=-0.3014, tangentMode=RCTM_User, time=0.7373, value=0.2044}
         ]}}}
      CurveIndex=$Particles.RibbonLinkOrder
     )
     B=&FloatFromCurve(
      FloatCurve=DataInterface() {propertyValues {Curve {keys [
          <at 0.2286 0.2496 RCIM_Cubic>
          {arriveTangent=-0.0519, interpMode=RCIM_Cubic, leaveTangent=-0.0519, tangentMode=RCTM_User, time=0.3281, value=0.38}
          {arriveTangent=-0.596, interpMode=RCIM_Cubic, leaveTangent=-0.596, tangentMode=RCTM_User, time=0.4854, value=0.4086}
          {arriveTangent=3.4042, interpMode=RCIM_Cubic, leaveTangent=3.4042, tangentMode=RCTM_User, time=0.6265, value=0.379}
          {arriveTangent=-0.3014, tangentMode=RCTM_User, time=0.7373, value=0.9203}
         ]}}}
      CurveIndex=$Particles.RibbonLinkOrder
     )
     Alpha=&FloatFromCurve(FloatCurve=DataInterface() {propertyValues {Curve {keys [<at 0.535 0>,<at 0.66 1>]}}}, CurveIndex=$System.Age)
    )) {module=ScaleRibbonWidth002}
   &ScaleColor(
    Scale Alpha=&FloatFromCurve(FloatCurve=DataInterface() {propertyValues {Curve {keys [<at 0.6448 1 RCIM_Cubic>,<at 0.7826 0>]}}}, CurveIndex=$System.Age)
    Scale RGB=&MakeVectorFromLinearColorRGB(Color=&FNTimeOfDayColor(Night Color=<rgba 0.7583 0.9628 1.4 1>, Morning Color=<rgba 1.7 1.7 1.7 1>, Day Color=<rgba 1 1 1 1>, Evening Color=<rgba 1.65 1.65 1.65 1>))
    ScaleA=true
    ScaleRGB=true
   )
  ]
  rendererClasses [&NiagaraRibbonRendererProperties]
  renderers [{rendererClass=&NiagaraRibbonRendererProperties}]
 }
}
summary {
 systemName=NS_Emote_SlopeTramp_Pour
 emitters [
  {emitterName=Stream, rendererClasses [&NiagaraRibbonRendererProperties]}
  {bEnabled=false, emitterName=Drops, rendererClasses [&NiagaraSpriteRendererProperties]}
 ]
 userVariables [
  User.bIsFrontEnd {type=NiagaraBool}
  User.bIsFrontEndPreview {type=NiagaraBool}
  User.RibbonIDOffset {type=NiagaraInt32}
  User.SkeletalMesh {
   type=NiagaraDataInterfaceSkeletalMesh
   defaultValue {
    struct=NiagaraExt_VariableValue_DataInterface
    value {dataInterface=&self:NiagaraDataInterfaceSkeletalMesh_0}
   }
  }
  User.VelocityIndx {type=NiagaraFloat, defaultValue=0, description=""}
 ]
}
