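# Niagara system for the pouring splashes in the SlopeTramp emote, built from a single sprite emitter, DropsStream.
# Burst and rate spawning emit short-lived drops from a small offset sphere, with curve-driven downward velocity, scale, spin and alpha fade.
# User.VelocityIndx switches between two velocity-scale presets, reshaping the stream's arc.
# Color is tinted by time of day, while dynamic material parameters animate over particle life.
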NiagaraSystem /BRCosmetics/Animation/Game/MainPlayer/Emotes/SlopeTramp/FX/NS_Emote_SlopeTramp_PourSplashes
dependencies {usedDataInterfaces [&NiagaraDataInterfaceCurve], usedDynamicInputs [], usedModules [&SystemState]}
emitters {DropsStream {
  emitterUpdateScript [
   &EmitterState(Life Cycle Mode=::NewEnumerator1, Loop Behavior=::NewEnumerator1, Loop Duration=0.3, UseLoopDelay=true)
   &SpawnBurst_Instantaneous(Spawn Count=2, Use Spawn Probability=false, UseLoopCountLimit=false)
   &SpawnRate(Spawn Probability=0.7, SpawnRate=10, Use Spawn Probability=false)
  ]
  particleSpawnScript [
   &InitializeParticle(
    Lifetime Max=0.43
    Lifetime Min=0.4
    Lifetime Mode=::NewEnumerator1
    Sprite Rotation Mode=::NewEnumerator2
    Sprite Size=&Multiply_Vector2D_ByFloat(
     Float=&UniformAOrBFloat(
      A=&RandomRangeFloat(Minimum=4, Maximum=5.5)
      B=6.5
      Distribution Weight=0
      Evaluation Type=::NewEnumerator0
      Recalculate Random Each Loop=true
      Override Seed=false
     )
     Vector2D=&RandomRangeVector2D(Minimum=<0.8 1.6>, Maximum=<0.9 2.6>)
    )
    Sprite Size Mode=::NewEnumerator4
    UsePositionOffset=false
    Write Material Random=true
   ) {module=InitializeParticle001}
   &DynamicMaterialParameters(
    Index 0 Param 2=1
    Index 0 Param 3=1
    Index 0 Param 4=1
    Module.Param0WriteEnabled=true
    Module.Param1WriteEnabled=false
    Module.Param2WriteEnabled=false
    Module.Param3WriteEnabled=false
   )
   &ShapeLocation(
    MaskSpawnGroup=false
    Offset=<-4 0 9>
    Offset Mode=::NewEnumerator0
    Override Seed=false
    Sphere Radius=2
    Sphere Surface Distribution=1
    UseOwnerScale=true
   )
   &AddVelocity(Velocity=&MakeVector(
     X=&Add_Float(
      A=0
      B=&Add_Float(
       A=&FloatFromCurve(FloatCurve=DataInterface() {propertyValues {Curve {keys [<at 0 1 RCIM_Cubic>,<at 1 0 RCIM_Cubic>]}}}, Scale Curve=-2)
       B=&FloatFromCurve(FloatCurve=DataInterface() {propertyValues {Curve {keys [<at 0 1 RCIM_Cubic>,<at 0.7673 0 RCIM_Cubic>]}}}, Scale Curve=-11.2)
      )
     )
     Y=1
     Z=&FloatFromCurve(
      FloatCurve=DataInterface() {propertyValues {Curve {keys [
          {arriveTangent=1.7154, interpMode=RCIM_Cubic, leaveTangent=1.7154, tangentMode=RCTM_User, time=0, value=0.5}
          {arriveTangent=0.4772, arriveTangentWeight=0.1, interpMode=RCIM_Cubic, leaveTangent=0.4772, tangentMode=RCTM_User, time=0.8, value=1}
         ]}}}
      Scale Curve=-175
     )
    )) {module=AddVelocity001}
   /Niagara/Modules/Update/SubUV/SubUVAnimation(SubUV Animation Mode=::NewEnumerator1) {module=SubUVAnimation001}
  ]
  particleUpdateScript [
   &ParticleState()
   &ScaleColor(
    Scale Alpha=&FloatFromCurve(FloatCurve=DataInterface() {propertyValues {Curve {keys [<at 0.7975 1 RCIM_Cubic>,<at 1 0>]}}})
    Scale RGB=&MakeVectorFromLinearColorRGB(Color=&FNTimeOfDayColor(Night Color=<rgba 0.7583 0.9628 1.4 1>, Morning Color=<rgba 1.7 1.7 1.7 1>, Day Color=<rgba 1 1 1 1>, Evening Color=<rgba 1.65 1.65 1.65 1>))
    ScaleA=true
    ScaleRGB=true
   ) {module=ScaleColor001}
   &ScaleSpriteSize(
    Uniform Curve Scale=0.8
    Uniform Curve Sprite Scale=DataInterface() {propertyValues {Curve {keys [
        {arriveTangent=7.64, interpMode=RCIM_Cubic, leaveTangent=7.64, tangentMode=RCTM_User, time=0, value=0}
        <at 0.2831 1 RCIM_Cubic>
        {arriveTangent=-2.8466, interpMode=RCIM_Cubic, leaveTangent=-2.8466, tangentMode=RCTM_User, time=1, value=0}
       ]}}}
   ) {module=ScaleSpriteSize001}
   &ScaleVelocity(Velocity Scale=&MakeVector(
     X=&FloatFromCurve(
      FloatCurve=DataInterface() {propertyValues {Curve {keys [
          <at 0.1053 4 RCIM_Cubic>
          {arriveTangent=-6.0023, interpMode=RCIM_Cubic, leaveTangent=-6.0023, tangentMode=RCTM_User, time=0.5552, value=1.0873}
          <at 1 0 RCIM_Cubic>
         ]}}}
      Scale Curve=&MergeHelper_FloatSwitch(
       0=0.7
       1=0.5
       2=?
       3=?
       Index=&MakeIntFromFloat(Float=$User.VelocityIndx, Conversion Mode=::NewEnumerator1)
       Num=::NewEnumerator0
      )
     )
     Y=1
     Z=&MergeHelper_FloatSwitch(
      0=0.72
      1=0.68
      2=?
      3=?
      Index=&MakeIntFromFloat(Float=$User.VelocityIndx, Conversion Mode=::NewEnumerator1)
      Num=::NewEnumerator0
     )
    ))
   &SolveForcesAndVelocity(Clamp Velocity=false, Limit Acceleration=false)
   &DynamicMaterialParameters(
    Index 0 Param 1=&RandomRangeFloat(Minimum=0, Maximum=1)
    Index 0 Param 2=&FloatFromCurve(FloatCurve=DataInterface() {propertyValues {Curve {keys [<at 0.5338 0.45 RCIM_Cubic>, {arriveTangent=0.9473, tangentMode=RCTM_User, time=1, value=0.9}]}}})
    Index 0 Param 3=&FloatFromCurve(FloatCurve=DataInterface() {propertyValues {Curve {keys [{arriveTangent=2.6651, interpMode=RCIM_Cubic, leaveTangent=2.6651, tangentMode=RCTM_User, time=0, value=0}, <at 1 1 RCIM_Cubic>]}}})
    Index 0 Param 4=&FloatFromCurve(FloatCurve=DataInterface() {propertyValues {Curve {keys [<at 0.5 0 RCIM_Cubic>, {arriveTangent=2.332, interpMode=RCIM_Cubic, leaveTangent=2.332, tangentMode=RCTM_User, time=0.75, value=1}]}}})
    Module.Param0WriteEnabled=true
    Module.Param1WriteEnabled=true
    Module.Param2WriteEnabled=false
    Module.Param3WriteEnabled=true
   ) {module=DynamicMaterialParameters002}
   &SpriteRotationRate(Rotation Rate=&Multiply_Float(
     A=&Multiply_Float(A=&RandomRangeFloat(Minimum=1, Maximum=0.35), B=$Particles.Random)
     B=&FloatFromCurve(
      FloatCurve=DataInterface() {propertyValues {Curve {keys [{arriveTangent=-2, interpMode=RCIM_Cubic, time=0.15, value=1}, {interpMode=RCIM_Cubic, leaveTangent=-8.0089, time=0.5, value=0}]}}}
      Scale Curve=800
     )
    ))
  ]
  rendererClasses [&NiagaraSpriteRendererProperties]
  renderers [{rendererClass=&NiagaraSpriteRendererProperties}]
 }}
summary {
 systemName=NS_Emote_SlopeTramp_PourSplashes
 emitters [{emitterName=DropsStream, rendererClasses [&NiagaraSpriteRendererProperties]}]
 userVariables [
  User.bIsFrontEnd {type=NiagaraBool}
  User.bIsFrontEndPreview {type=NiagaraBool}
  User.RibbonIDOffset {type=NiagaraInt32}
  User.VelocityIndx {type=NiagaraFloat, defaultValue=0, description=""}
 ]
}
