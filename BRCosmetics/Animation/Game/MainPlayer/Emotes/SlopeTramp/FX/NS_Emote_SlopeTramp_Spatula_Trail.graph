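# Niagara ribbon trail for the spatula in the SlopeTramp emote, built on the shared FN AnimTrail spawn, initialize and update modules.
# The trail follows a pair of bones or sockets as short-lived particles.
# Spawn color is a time-of-day tint, with alpha shaped by a curve over emitter loop age.
# Updates fade alpha by system age and a second curve, and taper ribbon width to zero.
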NiagaraSystem /BRCosmetics/Animation/Game/MainPlayer/Emotes/SlopeTramp/FX/NS_Emote_SlopeTramp_Spatula_Trail
dependencies {
 usedDataInterfaces [&NiagaraDataInterfaceColorCurve, &NiagaraDataInterfaceCurve, &NiagaraDataInterfaceSocketReader]
 usedDynamicInputs []
 usedModules [
  &self:SystemSpawnScript.SystemScriptSource.SystemScriptGraph.NiagaraNodeAssignment_0.SetVariables_#1
  &SystemState
 ]
}
emitters {Trail {
  emitterUpdateScript [
   &EmitterState()
   &FN_SpawnAnimTrail_v2(AnimTrailWidthMode=::FromSecond, Spawn Interval=0)
  ]
  particleSpawnScript [
   /Game/Effects/Niagara/Modules/AnimTrail/FN_InitializeAnimTrail/FN_InitializeAnimTrailBoneIndecies(Second Bone or Socket Index=1)
   &FN_InitializeAnimTrail(
    AnimTrailWidth=2
    Color=&MultiplyLinearColor(
     Linear Color=&FNTimeOfDayColor(
      Night Color=<rgba 0.0867 0.109 0.2969 1>
      Morning Color=<rgba 0.3094 0.3911 0.5885 1>
      Day Color=<rgba 1 1 1 1>
      Evening Color=<rgba 0.3094 0.3911 0.5885 1>
     )
     Scale Factor=&MakeVector4_FromLinearColor(LinearColor=&ColorFromCurve(
       ColorCurve=DataInterface() {propertyValues {
         AlphaCurve {keys [<at 0 0 RCIM_Cubic>,<at 0.065 0.15>,<at 1 0 RCIM_Cubic>]}
         BlueCurve {keys [<at 0 1 RCIM_Cubic>,<at 1 1 RCIM_Cubic>]}
         GreenCurve {keys [<at 0 1 RCIM_Cubic>,<at 1 1 RCIM_Cubic>]}
         RedCurve {keys [<at 0 1 RCIM_Cubic>,<at 1 1 RCIM_Cubic>]}
        }}
       CurveIndex=$Emitter.NormalizedLoopAge
      ))
    )
    Lifetime=0.2
   )
  ]
  particleUpdateScript [
   &FN_UpdateAnimTrail() {module=FN_UpdateAnimTrail001}
   &ScaleColor(
    Scale Alpha=&FloatFromCurve(FloatCurve=DataInterface() {propertyValues {Curve {keys [<at 0.2 1 RCIM_Cubic>,<at 0.6 0>]}}}, CurveIndex=$System.Age)
    ScaleA=true
    ScaleRGB=true
   ) {module=ScaleColor001}
   &ScaleColor(
    Scale Alpha=&FloatFromCurve(
     FloatCurve=DataInterface() {propertyValues {Curve {keys [
         {arriveTangent=7.5967, interpMode=RCIM_Cubic, leaveTangent=7.5967, tangentMode=RCTM_User, time=0, value=0}
         <at 0.0714 1 RCIM_Cubic>
         <at 1 0>
        ]}}}
     Scale Curve=0.7
    )
    ScaleA=true
    ScaleRGB=true
   )
   &ScaleRibbonWidth(Ribbon Width Scale=&FloatFromCurve(FloatCurve=DataInterface() {propertyValues {Curve {keys [<at 0 1>,<at 1 0>]}}})) {module=ScaleRibbonWidth001}
  ]
  rendererClasses [&NiagaraRibbonRendererProperties]
  renderers [{rendererClass=&NiagaraRibbonRendererProperties}]
 }}
summary {
 systemName=NS_Emote_SlopeTramp_Spatula_Trail
 emitters [{emitterName=Trail, rendererClasses [&NiagaraRibbonRendererProperties]}]
 userVariables [
  User.bIsFrontEnd {type=NiagaraBool}
  User.bIsFrontEndPreview {type=NiagaraBool}
  User.RibbonIDOffset {type=NiagaraInt32}
 ]
}
